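# List-row widget for an editor map QA static mesh audit, showing a mesh's instance count, actor count and asset name.
# OnListItemObjectSet reads a BP_StaticMeshAuditItem, caches its path and shows the name after the last dot, with the full path as tooltip.
# Two buttons run inline Python built from that path: sync the Content Browser, or select every level actor with a matching StaticMeshComponent.
# An event-less RowData text chain and debug prints remain.
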WidgetBlueprint /Game/NeverCook/EditorUtilities/QA/MapQA/EUW_QA_POIInstancePreviewWidget/WBP_StaticMeshAuditRow
graphs [EventGraph {nodes [
   n1 = Get!() -> (
    RowData_InstanceCount_2_#1:int
    RowData_ActorCount_4_#2:int
    RowData_AssetPath_7_#3:String
   )
   "SetText (Text)"(self=Get Txt_InstanceCount() -> (Txt_InstanceCount:Text Block Object Reference), InText="To Text (Integer)"(Value=@n1))
   > "SetText (Text)"(
    self=Get Txt_ActorCount() -> (Txt_ActorCount:Text Block Object Reference)
    InText="To Text (Integer)"(Value=@n1.RowData_ActorCount_4_#2)
   )
   > "SetText (Text)"(
    self=Get Txt_AssetPath() -> (Txt_AssetPath:Text Block Object Reference)
    InText="To Text (String)"(InString=@n1.RowData_AssetPath_7_#3)
   )
   "On Clicked (Btn_Browse)"() {is=title K2Node_ComponentBoundEvent}
   n2 = Get CurrentAssetPath() -> (CurrentAssetPath:String)
   > PrintString!(execute, InString=@n2, bPrintToScreen, bPrintToLog, TextColor=(R=1,G=0.0488,B=0,A=1), Duration, Key) -> (then)
   > ExecutePythonCommand(PythonCommand=Append!(A="import unreal; unreal.EditorAssetLibrary.sync_browser_to_objects(['", B=@n2, C:String="'.rsplit('.',1)[0]])") -> (ReturnValue))
   "On Clicked (Btn_SelectAllInstances)"() {is=title K2Node_ComponentBoundEvent}
   > ExecutePythonCommand(PythonCommand=Append(
     A="import unreal; p='"
     B=@n2
     C="'.rsplit('.',1)[0]; sel=[a for a in unreal.EditorLevelLibrary.get_all_level_actors() if any(c.static_mesh and c.static_mesh.get_path_name().rsplit('.',1)[0]==p for c in a.get_components_by_class(unreal.StaticMeshComponent))]; unreal.EditorLevelLibrary.set_selected_level_actors(sel)"
    ))
   n3 = Event OnListItemObjectSet()
   > n4 = Cast To BP_StaticMeshAuditItem!(execute, Object=@n3.ListItemObject) -> (then, CastFailed, AsBP Static Mesh Audit Item:BP Static Mesh Audit Item Object Reference) {is=title K2Node_DynamicCast}
   > n5 = Set CurrentAssetPath(CurrentAssetPath:String=Get AssetPath!(self:BP Static Mesh Audit Item Object Reference=@n4) -> (AssetPath:String)) -> (Output_Get:String)
   > PrintString(InString=@n5, TextColor=(R=0,G=1,B=0.1273,A=1))
   > "SetText (Text)"(
    self=Get Txt_InstanceCount()
    InText="To Text (Integer)"(Value=Get InstanceCount!(self:BP Static Mesh Audit Item Object Reference=@n4) -> (InstanceCount:int=0))
   )
   > "SetText (Text)"(
    self=Get Txt_ActorCount()
    InText="To Text (Integer)"(Value=Get ActorCount!(self:BP Static Mesh Audit Item Object Reference=@n4) -> (ActorCount:int=0))
   )
   n6 = Get AssetPath(self=@n4)
   > n7 = Increment Int(Value=FindSubstring(SearchIn=@n6, Substring=., bSearchFromEnd=true))
   n8 = Get Txt_AssetPath()
   > "SetText (Text)"(self=@n8, InText="To Text (String)"(InString=RightChop!(SourceString:String=@n6, Count:int=@n7) -> (ReturnValue:String)))
   > SetToolTipText!(execute, self:Widget Object Reference=@n8, InToolTipText:"Text (by ref)"="To Text (String)"(InString=@n6)) -> (then)
   PrintString(InString=Get Class Display Name(Class=Get Class(Object=@n4) -> (ReturnValue:BP Static Mesh Audit Item Class Reference)))
   Event PreConstruct()
   Event Construct()
   Event Tick()
   Get CurrentAssetPath()
  ]}]
variables [
 CurrentAssetPath {type=string, default=""}
 RowData {type=S_StaticMeshAuditRow, default {actorCount=0, assetPath="", instanceCount=0}}
]
widgets {
 info {inheritedWidgetCount=0, parentClass=&UserWidget, rootWidgetClass=&HorizontalBox, widgetCount=12}
 widgets [
  HorizontalBox_28 {widgetClassPath=&HorizontalBox, widget=HorizontalBox_28}
  Border_55 {
   widgetClassPath=&Border
   parent=HorizontalBox_28
   widget=Border_55
   properties {horizontalAlignment=HAlign_Left, verticalAlignment=VAlign_Center, brushColor=<rgba 0.0729 0.0729 0.0729 1>, padding=<ltrb 0 0 0 0>}
   slot {class=HorizontalBoxSlot, properties {horizontalAlignment=HAlign_Left, verticalAlignment=VAlign_Center, padding=<ltrb 4 4 4 4>}}
  }
  Txt_InstanceCount {
   widgetClassPath=&UEFN_TextBlock
   bIsVariable=true
   parent=Border_55
   widget=Txt_InstanceCount
   properties {
    text=Text Block
    font {
     bForceMonospaced=false
     bMaterialIsStencil=false
     fontMaterial=null
     fontObject=&NotoSans
     letterSpacing=0
     monospacedWidth=1
     size=9
     skewAmount=0
     typefaceFontName=NotoSans-Bold
     outlineSettings {
      bApplyOutlineToDropShadows=false
      bMiteredCorners=false
      bSeparateFillAlpha=false
      outlineMaterial=null
      outlineSize=0
      outlineColor=<rgba 0 0 0 1>
     }
    }
   }
   slot {class=BorderSlot, properties {horizontalAlignment=HAlign_Left, verticalAlignment=VAlign_Center, padding=<ltrb 0 0 0 0>}}
  }
  Border {
   widgetClassPath=&Border
   parent=HorizontalBox_28
   widget=Border
   properties {horizontalAlignment=HAlign_Left, verticalAlignment=VAlign_Center, brushColor=<rgba 0 0 0 1>, padding=<ltrb 0 0 0 0>}
   slot {class=HorizontalBoxSlot, properties {horizontalAlignment=HAlign_Left, verticalAlignment=VAlign_Center, padding=<ltrb 4 4 4 4>}}
  }
  Txt_ActorCount {
   widgetClassPath=&UEFN_TextBlock
   bIsVariable=true
   parent=Border
   widget=Txt_ActorCount
   properties {
    text=Text Block
    colorAndOpacity {colorUseRule=UseColor_Specified, specifiedColor=<rgba 0.0417 0.2946 1 1>}
    font {
     bForceMonospaced=false
     bMaterialIsStencil=false
     fontMaterial=null
     fontObject=&NotoSans
     letterSpacing=0
     monospacedWidth=1
     size=9
     skewAmount=0
     typefaceFontName=NotoSans-Bold
     outlineSettings {
      bApplyOutlineToDropShadows=false
      bMiteredCorners=false
      bSeparateFillAlpha=false
      outlineMaterial=null
      outlineSize=0
      outlineColor=<rgba 0 0 0 1>
     }
    }
   }
   slot {class=BorderSlot, properties {horizontalAlignment=HAlign_Left, verticalAlignment=VAlign_Center, padding=<ltrb 0 0 0 0>}}
  }
  Border_1 {
   widgetClassPath=&Border
   parent=HorizontalBox_28
   widget=Border_1
   properties {
    background {
     bIsDynamicallyLoaded=false
     drawAs=Image
     imageType=NoImage
     mirroring=NoMirror
     resourceName=None
     resourceObject=null
     tiling=NoTile
     imageSize=<32 32>
     margin=<ltrb 0 0 0 0>
     outlineSettings {
      bUseBrushTransparency=false
      roundingType=HalfHeightRadius
      width=0
      color {colorUseRule=UseColor_Specified, specifiedColor=<rgba 0 0 0 0>}
      cornerRadii=<0 0 0 0>
     }
     tintColor {colorUseRule=UseColor_Specified, specifiedColor=<rgba 1 1 1 0>}
     uVRegion {bIsValid=false, max=<0 0>, min=<0 0>}
    }
    brushColor=<rgba 0.0573 0.0573 0.0573 1>
    padding=<ltrb 0 0 0 0>
   }
   slot {
    class=HorizontalBoxSlot
    properties {horizontalAlignment=HAlign_Left, verticalAlignment=VAlign_Center, padding=<ltrb 4 4 4 4>, size {sizeRule=Fill, value=1}}
   }
  }
  Txt_AssetPath {
   widgetClassPath=&UEFN_TextBlock
   bIsVariable=true
   parent=Border_1
   widget=Txt_AssetPath
   properties {
    text=Text Block
    textOverflowPolicy=Ellipsis
    colorAndOpacity {colorUseRule=UseColor_Specified, specifiedColor=<rgba 0.4487 0.2865 1 1>}
    font {
     bForceMonospaced=false
     bMaterialIsStencil=false
     fontMaterial=null
     fontObject=&NotoSans
     letterSpacing=0
     monospacedWidth=1
     size=9
     skewAmount=0
     typefaceFontName=NotoSans-Bold
     outlineSettings {
      bApplyOutlineToDropShadows=false
      bMiteredCorners=false
      bSeparateFillAlpha=false
      outlineMaterial=null
      outlineSize=0
      outlineColor=<rgba 0 0 0 1>
     }
    }
   }
   slot {class=BorderSlot, properties {padding=<ltrb 0 0 0 0>}}
  }
  Spacer_94 {
   widgetClassPath=&Spacer
   parent=HorizontalBox_28
   widget=Spacer_94
   slot {class=HorizontalBoxSlot, properties {size {sizeRule=Fill, value=1}}}
  }
  Btn_Browse {
   widgetClassPath=&Button
   parent=HorizontalBox_28
   widget=Btn_Browse
   properties {widgetStyle {
     clickedSlateSound {resourceObject=null}
     disabled {
      bIsDynamicallyLoaded=false
      drawAs=NoDrawType
      imageType=NoImage
      mirroring=NoMirror
      resourceName=None
      resourceObject=null
      tiling=NoTile
      imageSize=<0 0>
      margin=<ltrb 0 0 0 0>
      outlineSettings {
       bUseBrushTransparency=false
       roundingType=HalfHeightRadius
       width=0
       color {colorUseRule=UseColor_Specified, specifiedColor=<rgba 0 0 0 0>}
       cornerRadii=<0 0 0 0>
      }
      tintColor {colorUseRule=UseColor_Specified, specifiedColor=<rgba 1 1 1 1>}
      uVRegion {bIsValid=false, max=<0 0>, min=<0 0>}
     }
     disabledForeground {colorUseRule=UseColor_Specified, specifiedColor=<rgba 0.5271 0.5271 0.5271 1>}
     hovered {
      bIsDynamicallyLoaded=false
      drawAs=RoundedBox
      imageType=NoImage
      mirroring=NoMirror
      resourceName=None
      resourceObject=null
      tiling=NoTile
      imageSize=<32 32>
      margin=<ltrb 0 0 0 0>
      outlineSettings {
       bUseBrushTransparency=true
       roundingType=FixedRadius
       width=1
       color {colorUseRule=UseColor_Specified, specifiedColor=<rgba 0.7243 0.7243 0.7243 1>}
       cornerRadii=<4 4 4 4>
      }
      tintColor {colorUseRule=UseColor_Specified, specifiedColor=<rgba 0.7243 0.7243 0.7243 1>}
      uVRegion {bIsValid=false, max=<0 0>, min=<0 0>}
     }
     hoveredForeground {colorUseRule=UseColor_Specified, specifiedColor=<rgba 0 1 1 1>}
     hoveredSlateSound {resourceObject=null}
     normal {
      bIsDynamicallyLoaded=false
      drawAs=RoundedBox
      imageType=NoImage
      mirroring=NoMirror
      resourceName=None
      resourceObject=null
      tiling=NoTile
      imageSize=<32 32>
      margin=<ltrb 0 0 0 0>
      outlineSettings {
       bUseBrushTransparency=true
       roundingType=FixedRadius
       width=1
       color {colorUseRule=UseColor_Specified, specifiedColor=<rgba 0.6951 0.6951 0.6951 1>}
       cornerRadii=<4 4 4 4>
      }
      tintColor {colorUseRule=UseColor_Specified, specifiedColor=<rgba 0.0395 0.0395 0.0395 1>}
      uVRegion {bIsValid=false, max=<0 0>, min=<0 0>}
     }
     normalForeground {colorUseRule=UseColor_Specified, specifiedColor=<rgba 0 1 1 1>}
     normalPadding=<ltrb 12 1.5 12 1.5>
     pressed {
      bIsDynamicallyLoaded=false
      drawAs=RoundedBox
      imageType=NoImage
      mirroring=NoMirror
      resourceName=None
      resourceObject=null
      tiling=NoTile
      imageSize=<32 32>
      margin=<ltrb 0 0 0 0>
      outlineSettings {
       bUseBrushTransparency=true
       roundingType=FixedRadius
       width=1
       color {colorUseRule=UseColor_Specified, specifiedColor=<rgba 0.7243 0.7243 0.7243 1>}
       cornerRadii=<4 4 4 4>
      }
      tintColor {colorUseRule=UseColor_Specified, specifiedColor=<rgba 0.3843 0.3843 0.3843 1>}
      uVRegion {bIsValid=false, max=<0 0>, min=<0 0>}
     }
     pressedForeground {colorUseRule=UseColor_Specified, specifiedColor=<rgba 0 1 1 1>}
     pressedPadding=<ltrb 12 1.5 12 1.5>
     pressedSlateSound {resourceObject=null}
    }}
   slot {class=HorizontalBoxSlot, properties {horizontalAlignment=HAlign_Right, verticalAlignment=VAlign_Center}}
  }
  UEFN_TextBlock_C_55 {
   widgetClassPath=&UEFN_TextBlock
   parent=Btn_Browse
   widget=UEFN_TextBlock_C_55
   properties {
    text=Browse To Asset
    toolTipText=Browse to the mesh in the Content Browser
    font {
     bForceMonospaced=false
     bMaterialIsStencil=false
     fontMaterial=null
     fontObject=&NotoSans
     letterSpacing=0
     monospacedWidth=1
     size=9
     skewAmount=0
     typefaceFontName=NotoSans-Bold
     outlineSettings {
      bApplyOutlineToDropShadows=false
      bMiteredCorners=false
      bSeparateFillAlpha=false
      outlineMaterial=null
      outlineSize=0
      outlineColor=<rgba 0 0 0 1>
     }
    }
   }
   slot {class=ButtonSlot}
  }
  Btn_SelectAllInstances {
   widgetClassPath=&Button
   parent=HorizontalBox_28
   widget=Btn_SelectAllInstances
   properties {widgetStyle {
     clickedSlateSound {resourceObject=null}
     disabled {
      bIsDynamicallyLoaded=false
      drawAs=NoDrawType
      imageType=NoImage
      mirroring=NoMirror
      resourceName=None
      resourceObject=null
      tiling=NoTile
      imageSize=<0 0>
      margin=<ltrb 0 0 0 0>
      outlineSettings {
       bUseBrushTransparency=false
       roundingType=HalfHeightRadius
       width=0
       color {colorUseRule=UseColor_Specified, specifiedColor=<rgba 0 0 0 0>}
       cornerRadii=<0 0 0 0>
      }
      tintColor {colorUseRule=UseColor_Specified, specifiedColor=<rgba 1 1 1 1>}
      uVRegion {bIsValid=false, max=<0 0>, min=<0 0>}
     }
     disabledForeground {colorUseRule=UseColor_Specified, specifiedColor=<rgba 0.5271 0.5271 0.5271 1>}
     hovered {
      bIsDynamicallyLoaded=false
      drawAs=RoundedBox
      imageType=NoImage
      mirroring=NoMirror
      resourceName=None
      resourceObject=null
      tiling=NoTile
      imageSize=<32 32>
      margin=<ltrb 0 0 0 0>
      outlineSettings {
       bUseBrushTransparency=true
       roundingType=FixedRadius
       width=1
       color {colorUseRule=UseColor_Specified, specifiedColor=<rgba 0.7243 0.7243 0.7243 1>}
       cornerRadii=<4 4 4 4>
      }
      tintColor {colorUseRule=UseColor_Specified, specifiedColor=<rgba 0.7243 0.7243 0.7243 1>}
      uVRegion {bIsValid=false, max=<0 0>, min=<0 0>}
     }
     hoveredForeground {colorUseRule=UseColor_Specified, specifiedColor=<rgba 0 1 1 1>}
     hoveredSlateSound {resourceObject=null}
     normal {
      bIsDynamicallyLoaded=false
      drawAs=RoundedBox
      imageType=NoImage
      mirroring=NoMirror
      resourceName=None
      resourceObject=null
      tiling=NoTile
      imageSize=<32 32>
      margin=<ltrb 0 0 0 0>
      outlineSettings {
       bUseBrushTransparency=true
       roundingType=FixedRadius
       width=1
       color {colorUseRule=UseColor_Specified, specifiedColor=<rgba 0.6951 0.6951 0.6951 1>}
       cornerRadii=<4 4 4 4>
      }
      tintColor {colorUseRule=UseColor_Specified, specifiedColor=<rgba 0.0395 0.0395 0.0395 1>}
      uVRegion {bIsValid=false, max=<0 0>, min=<0 0>}
     }
     normalForeground {colorUseRule=UseColor_Specified, specifiedColor=<rgba 0 1 1 1>}
     normalPadding=<ltrb 12 1.5 12 1.5>
     pressed {
      bIsDynamicallyLoaded=false
      drawAs=RoundedBox
      imageType=NoImage
      mirroring=NoMirror
      resourceName=None
      resourceObject=null
      tiling=NoTile
      imageSize=<32 32>
      margin=<ltrb 0 0 0 0>
      outlineSettings {
       bUseBrushTransparency=true
       roundingType=FixedRadius
       width=1
       color {colorUseRule=UseColor_Specified, specifiedColor=<rgba 0.7243 0.7243 0.7243 1>}
       cornerRadii=<4 4 4 4>
      }
      tintColor {colorUseRule=UseColor_Specified, specifiedColor=<rgba 0.3843 0.3843 0.3843 1>}
      uVRegion {bIsValid=false, max=<0 0>, min=<0 0>}
     }
     pressedForeground {colorUseRule=UseColor_Specified, specifiedColor=<rgba 0 1 1 1>}
     pressedPadding=<ltrb 12 1.5 12 1.5>
     pressedSlateSound {resourceObject=null}
    }}
   slot {class=HorizontalBoxSlot, properties {horizontalAlignment=HAlign_Right, verticalAlignment=VAlign_Center, padding=<ltrb 10 10 10 10>}}
  }
  UEFN_TextBlock_C_61 {
   widgetClassPath=&UEFN_TextBlock
   parent=Btn_SelectAllInstances
   widget=UEFN_TextBlock_C_61
   properties {
    text=Select All Instances
    toolTipText=Select All Instances of this Mesh in the Loaded POI
    font {
     bForceMonospaced=false
     bMaterialIsStencil=false
     fontMaterial=null
     fontObject=&NotoSans
     letterSpacing=0
     monospacedWidth=1
     size=9
     skewAmount=0
     typefaceFontName=NotoSans-Bold
     outlineSettings {
      bApplyOutlineToDropShadows=false
      bMiteredCorners=false
      bSeparateFillAlpha=false
      outlineMaterial=null
      outlineSize=0
      outlineColor=<rgba 0 0 0 1>
     }
    }
   }
   slot {class=ButtonSlot}
  }
 ]
}
