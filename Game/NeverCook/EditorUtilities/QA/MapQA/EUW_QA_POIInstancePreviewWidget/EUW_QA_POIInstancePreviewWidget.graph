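# Editor Utility Widget for map QA that lists static mesh usage, with instance and actor counts, across selected POI actors and their attached children.
# The generate button runs embedded Python that writes CSV and TXT audit files to the Saved folder.
# After a short delay, Blueprint parses the pipe-separated CSV rows into BP_StaticMeshAuditItem objects for the list view.
# A second button reveals the CSV in Explorer; the TXT-to-ResultsBox chain is unwired.
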EditorUtilityWidgetBlueprint /Game/NeverCook/EditorUtilities/QA/MapQA/EUW_QA_POIInstancePreviewWidget/EUW_QA_POIInstancePreviewWidget
graphs [EventGraph {nodes [
   n1 = Read File to String!(
    execute
    FilePath:String=Append(A=GetProjectSavedDirectory!() -> (ReturnValue:String), B=/StaticMeshAudit_Selected.txt)
    bUseFullFilePath:bool=true
   ) -> (then, ResultOut:String, ReturnValue:bool)
   > "SetText (Multi-Line Editable Text)"!(
    execute
    self:"Editable Text (Multi- Line) Object Reference"=Get ResultsBox() -> (ResultsBox:Editor Utility Multi Line Editable Text Object Reference)
    InText:Text="To Text (String)"(InString=@n1)
   ) -> (then)
   "On Clicked (WriteSelectedInstances)"() {is=title K2Node_ComponentBoundEvent}
   > SetIsEnabled(self=Get GoToCSV() -> (GoToCSV:Editor Utility Button Object Reference), bInIsEnabled=false)
   > ExecutePythonCommand(PythonCommand="import unreal\r\nfrom collections import defaultdict\r\nimport os\r\n\r\nOUTPUT_CSV = \"StaticMeshAudit_Selected.csv\"\r\nOUTPUT_TXT = \"StaticMeshAudit_Selected.txt\"\r\nINCLUDE_ATTACHED_ACTORS = True\r\n\r\ndef asset_path(obj):\r\n    if not obj:\r\n        return None\r\n    try:\r\n        return obj.get_path_name()\r\n    except Exception:\r\n        return str(obj)\r\n\r\ndef get_instance_count(comp):\r\n    for fn in (\"get_instance_count\", \"get_num_instances\"):\r\n        if hasattr(comp, fn):\r\n            try:\r\n                return int(getattr(comp, fn)())\r\n            except Exception:\r\n                pass\r\n    for prop in (\"instance_count\", \"num_instances\"):\r\n        if hasattr(comp, prop):\r\n            try:\r\n                return int(getattr(comp, prop))\r\n            except Exception:\r\n                pass\r\n    return 0\r\n\r\ndef gather_selected_actors(include_attached=True):\r\n    selected = unreal.EditorLevelLibrary.get_selected_level_actors()\r\n    if not include_attached:\r\n        return list(selected)\r\n\r\n    result = set(selected)\r\n    stack = list(selected)\r\n    while stack:\r\n        a = stack.pop()\r\n        try:\r\n            attached = a.get_attached_actors()\r\n        except Exception:\r\n            attached = []\r\n        for child in attached:\r\n            if child not in result:\r\n                result.add(child)\r\n                stack.append(child)\r\n    return list(result)\r\n\r\ndef scan_actors_static_meshes(actors):\r\n    counts = defaultdict(int)\r\n    actor_sets = defaultdict(set)\r\n\r\n    for a in actors:\r\n        actor_name = a.get_name()\r\n\r\n        try:\r\n            for c in a.get_components_by_class(unreal.StaticMeshComponent):\r\n                mesh = c.static_mesh if hasattr(c, \"static_mesh\") else c.get_static_mesh()\r\n                p = asset_path(mesh)\r\n                if p:\r\n                    counts[p] += 1\r\n                    actor_sets[p].add(actor_name)\r\n        except Exception:\r\n            pass\r\n\r\n        try:\r\n            for c in a.get_components_by_class(unreal.InstancedStaticMeshComponent):\r\n                mesh = c.static_mesh if hasattr(c, \"static_mesh\") else c.get_static_mesh()\r\n                p = asset_path(mesh)\r\n                if p:\r\n                    counts[p] += get_instance_count(c)\r\n                    actor_sets[p].add(actor_name)\r\n        except Exception:\r\n            pass\r\n\r\n        try:\r\n            for c in a.get_components_by_class(unreal.HierarchicalInstancedStaticMeshComponent):\r\n                mesh = c.static_mesh if hasattr(c, \"static_mesh\") else c.get_static_mesh()\r\n                p = asset_path(mesh)\r\n                if p:\r\n                    counts[p] += get_instance_count(c)\r\n                    actor_sets[p].add(actor_name)\r\n        except Exception:\r\n            pass\r\n\r\n    return counts, actor_sets\r\n\r\ndef write_files(counts, actor_sets, scanned_actor_count):\r\n    saved_dir = unreal.Paths.project_saved_dir()\r\n    csv_path = os.path.join(saved_dir, OUTPUT_CSV)\r\n    txt_path = os.path.join(saved_dir, OUTPUT_TXT)\r\n    os.makedirs(saved_dir, exist_ok=True)\r\n\r\n    ## CSV\r\n    csv_lines = [\"AssetPath,InstanceCount,ActorCount\"]\r\n    items = sorted(counts.items(), key=lambda kv: kv[1], reverse=True)\r\n    for path, inst in items:\r\n        csv_lines.append(\"{},{},{}\".format(path, inst, len(actor_sets[path])))\r\n    with open(csv_path, \"w\", encoding=\"utf-8\", newline=\"\") as f:\r\n        f.write(\"|\".join(csv_lines))\r\n\r\n    ## Pretty TXT for your widget\r\n    txt_lines = []\r\n    txt_lines.append(\"Static Mesh Audit (Selected Actors)\")\r\n    txt_lines.append(\"Actors scanned: {}\".format(scanned_actor_count))\r\n    txt_lines.append(\"Unique meshes: {}\".format(len(counts)))\r\n    txt_lines.append(\"\")\r\n    txt_lines.append(\"{:<8} {:<8} {}\".format(\"INST\", \"ACTORS\", \"ASSET\"))\r\n    txt_lines.append(\"-\" * 80)\r\n    for path, inst in items:\r\n        txt_lines.append(\"{:<8} {:<8} {}\".format(inst, len(actor_sets[path]), path))\r\n\r\n    with open(txt_path, \"w\", encoding=\"utf-8\", newline=\"\") as f:\r\n        f.write(\"\\n\".join(txt_lines))\r\n\r\n    unreal.log(\"Wrote CSV: {}\".format(csv_path))\r\n    unreal.log(\"Wrote TXT: {}\".format(txt_path))\r\n\r\ndef run():\r\n    actors = gather_selected_actors(include_attached=INCLUDE_ATTACHED_ACTORS)\r\n    if not actors:\r\n        unreal.log_warning(\"No Outliner selection.\")\r\n        return\r\n\r\n    counts, actor_sets = scan_actors_static_meshes(actors)\r\n    write_files(counts, actor_sets, len(actors))\r\n\r\nrun()\r\n")
   > Delay!(execute, WorldContextObject:Object Reference, Duration=0.2) -> (then)
   > n2 = Read File to String(FilePath=Append(A=GetProjectSavedDirectory(), B=/StaticMeshAudit_Selected.csv), bUseFullFilePath=true)
   > n3 = Set Lines(Lines:Array of Strings=ParseIntoArray(SourceString=@n2, Delimiter=|)) -> (Output_Get:Array of Strings)
   > Clear(TargetArray:Array of BP Static Mesh Audit Item Object References=Get AuditItems() -> (AuditItems:Array of BP Static Mesh Audit Item Object References))
   > n4 = For Each Loop(Array=@n3)
   PrintString(execute=@n4.Completed, InString=AuditItems, TextColor=(R=1,G=0.0448,B=0.2051,A=1))
   n5 = Get AuditListView() -> (AuditListView:Editor Utility List View Object Reference)
   > ClearListItems(self=@n5)
   > Set List Items!(execute, self:List View Object Reference=@n5, InListItems:Array of Object References=Get AuditItems()) -> (then) {is=title K2Node_CallFunction}
   > RequestRefresh(self:List View Base Object Reference=@n5)
   > SetIsEnabled(self=Get GoToCSV(), bInIsEnabled=true)
   n6 = StartsWith(SourceString=@n4, InPrefix=AssetPath)
   n7 = IsEmpty(InString=@n4)
   n8 = Branch(execute=@n4.LoopBody, Condition=AND Boolean(A=NOT Boolean(A=@n6), B=NOT Boolean(A=@n7)))
   > n9 = Construct Object from Class(Class=&/BP_StaticMeshAuditItem.BP_StaticMeshAuditItem_C) -> (ReturnValue:BP Static Mesh Audit Item Object Reference)
   n10 = ParseIntoArray(SourceString=@n4, Delimiter=",")
   n11 = "Get (a copy)"(Array=@n10)
   > n12 = Set AssetPath!(execute, AssetPath:String=@n11, self:BP Static Mesh Audit Item Object Reference=@n9) -> (then, Output_Get:String)
   > PrintString(InString=@n12, TextColor=(R=1,G=0.0448,B=0.2051,A=1))
   > n13 = Set InstanceCount!(
    execute
    InstanceCount:int=String To Integer(InString="Get (a copy)"(Array=@n10, Dimension 1=1))
    self:BP Static Mesh Audit Item Object Reference=@n9
   ) -> (then, Output_Get:int=0)
   > PrintString(InString="To String (Integer)"(InInt=@n13), TextColor=(R=1,G=0.0448,B=0.2051,A=1))
   > n14 = Set ActorCount!(
    execute
    ActorCount:int=String To Integer(InString="Get (a copy)"(Array=@n10, Dimension 1=2))
    self:BP Static Mesh Audit Item Object Reference=@n9
   ) -> (then, Output_Get:int=0)
   > PrintString(InString="To String (Integer)"(InInt=@n14), TextColor=(R=1,G=0.0448,B=0.2051,A=1))
   > Add(
    TargetArray:Array of BP Static Mesh Audit Item Object References=Get AuditItems()
    NewItem:"BP Static Mesh Audit Item Object Reference (by ref)"=@n9
   ) {is=title K2Node_CallArrayFunction}
   PrintString(execute=@n8.else, InString=False, TextColor=(R=1,G=0.0448,B=0.2051,A=1))
   "On Clicked (GoToCSV)"() {is=title K2Node_ComponentBoundEvent}
   > ExecutePythonCommand(PythonCommand="import subprocess, unreal; subprocess.Popen(['explorer', '/select,', (unreal.Paths.project_saved_dir() +\r\n  'StaticMeshAudit_Selected.csv').replace('/', '\\\\')])")
   PrintString(InString=@n11)
  ]}]
variables [
 AuditItems {type="array<BP_StaticMeshAuditItem_C>", default []}
 AuditRows {type="array<S_StaticMeshAuditRow>", default []}
 Lines {type="array<string>", default []}
]
widgets {
 info {rootWidgetClass=&CanvasPanel, widgetCount=20}
 widgets [
  CanvasPanel_1 {widgetClassPath=&CanvasPanel, widget=CanvasPanel_1}
  GridPanel_0 {
   widgetClassPath=&GridPanel
   parent=CanvasPanel_1
   widget=GridPanel_0
   slot {class=CanvasPanelSlot, properties {layoutData {alignment=<0 0>, anchors {maximum=<0 0>, minimum=<0 0>}, offsets=<ltrb 0 0 1248 724>}}}
  }
  HorizontalBox_170 {widgetClassPath=&HorizontalBox, parent=GridPanel_0, widget=HorizontalBox_170, slot {class=GridSlot, properties {column=3, row=7}}}
  Border_129 {
   widgetClassPath=&Border
   parent=HorizontalBox_170
   widget=Border_129
   properties {padding=<ltrb 0 0 0 0>}
   slot {class=HorizontalBoxSlot}
  }
  HorizontalBox_0 {
   widgetClassPath=&HorizontalBox
   parent=Border_129
   widget=HorizontalBox_0
   properties {toolTipText=Number of Instances Total Per asset}
   slot {class=BorderSlot, properties {padding=<ltrb 0 0 0 0>}}
  }
  UEFN_TextBlock_C_0 {
   widgetClassPath=&UEFN_TextBlock
   parent=HorizontalBox_0
   widget=UEFN_TextBlock_C_0
   properties {
    text=Instance Count
    colorAndOpacity {colorUseRule=UseColor_Specified, specifiedColor=<rgba 0 0 0 1>}
    font {
     bForceMonospaced=false
     bMaterialIsStencil=false
     fontMaterial=null
     fontObject=&BurbankSmall
     letterSpacing=0
     monospacedWidth=1
     size=18
     skewAmount=0
     typefaceFontName=Black
     outlineSettings {
      bApplyOutlineToDropShadows=false
      bMiteredCorners=false
      bSeparateFillAlpha=false
      outlineMaterial=null
      outlineSize=0
      outlineColor=<rgba 0 0 0 1>
     }
    }
   }
   slot {class=HorizontalBoxSlot, properties {horizontalAlignment=HAlign_Left, verticalAlignment=VAlign_Top, padding=<ltrb 10 10 10 10>}}
  }
  Border_61 {
   widgetClassPath=&Border
   parent=HorizontalBox_170
   widget=Border_61
   properties {brushColor=<rgba 0.0423 0.2961 1 1>, padding=<ltrb 0 0 0 0>}
   slot {class=HorizontalBoxSlot}
  }
  HorizontalBox_106 {
   widgetClassPath=&HorizontalBox
   parent=Border_61
   widget=HorizontalBox_106
   properties {toolTipText="Number of Actors Total Per asset (ex: actor count may be less if there are two instances within the same actor)"}
   slot {class=BorderSlot, properties {padding=<ltrb 0 0 0 0>}}
  }
  UEFN_TextBlock_C {
   widgetClassPath=&UEFN_TextBlock
   parent=HorizontalBox_106
   widget=UEFN_TextBlock_C
   properties {
    text=Actor Count
    colorAndOpacity {colorUseRule=UseColor_Specified, specifiedColor=<rgba 0 0 0 1>}
    font {
     bForceMonospaced=false
     bMaterialIsStencil=false
     fontMaterial=null
     fontObject=&BurbankSmall
     letterSpacing=0
     monospacedWidth=1
     size=18
     skewAmount=0
     typefaceFontName=Black
     outlineSettings {
      bApplyOutlineToDropShadows=false
      bMiteredCorners=false
      bSeparateFillAlpha=false
      outlineMaterial=null
      outlineSize=0
      outlineColor=<rgba 0 0 0 1>
     }
    }
   }
   slot {class=HorizontalBoxSlot, properties {horizontalAlignment=HAlign_Left, verticalAlignment=VAlign_Top, padding=<ltrb 10 10 10 10>}}
  }
  Border_101 {
   widgetClassPath=&Border
   parent=HorizontalBox_170
   widget=Border_101
   properties {brushColor=<rgba 0.275 0.0611 1 1>, padding=<ltrb 0 0 0 0>}
   slot {class=HorizontalBoxSlot}
  }
  HorizontalBox_129 {
   widgetClassPath=&HorizontalBox
   parent=Border_101
   widget=HorizontalBox_129
   properties {toolTipText=Static Mesh used in the actor}
   slot {class=BorderSlot, properties {padding=<ltrb 0 0 0 0>}}
  }
  UEFN_TextBlock_C_1 {
   widgetClassPath=&UEFN_TextBlock
   parent=HorizontalBox_129
   widget=UEFN_TextBlock_C_1
   properties {
    text=Actor Path
    colorAndOpacity {colorUseRule=UseColor_Specified, specifiedColor=<rgba 0 0 0 1>}
    font {
     bForceMonospaced=false
     bMaterialIsStencil=false
     fontMaterial=null
     fontObject=&BurbankSmall
     letterSpacing=0
     monospacedWidth=1
     size=18
     skewAmount=0
     typefaceFontName=Black
     outlineSettings {
      bApplyOutlineToDropShadows=false
      bMiteredCorners=false
      bSeparateFillAlpha=false
      outlineMaterial=null
      outlineSize=0
      outlineColor=<rgba 0 0 0 1>
     }
    }
   }
   slot {class=HorizontalBoxSlot, properties {horizontalAlignment=HAlign_Left, verticalAlignment=VAlign_Top, padding=<ltrb 10 10 10 10>}}
  }
  ResultsBox {
   widgetClassPath=/Script/Blutility.EditorUtilityMultiLineEditableText
   bIsVariable=true
   parent=GridPanel_0
   widget=ResultsBox
   slot {class=GridSlot, properties {column=3, horizontalAlignment=HAlign_Center, row=21, verticalAlignment=VAlign_Center}}
  }
  SizeBox_0 {
   widgetClassPath=&SizeBox
   parent=GridPanel_0
   widget=SizeBox_0
   properties {bOverride_HeightOverride=true, heightOverride=800}
   slot {class=GridSlot, properties {column=3, row=21}}
  }
  AuditListView {
   widgetClassPath=/Script/Blutility.EditorUtilityListView
   bIsVariable=true
   parent=SizeBox_0
   widget=AuditListView
   properties {
    entryWidgetClass=&/WBP_StaticMeshAuditRow.WBP_StaticMeshAuditRow_C
    visibility=SelfHitTestInvisible
   }
   slot {class=SizeBoxSlot}
  }
  HorizontalBox_1 {
   widgetClassPath=&HorizontalBox
   parent=GridPanel_0
   widget=HorizontalBox_1
   slot {class=GridSlot, properties {column=3, horizontalAlignment=HAlign_Left, row=6, verticalAlignment=VAlign_Top}}
  }
  WriteSelectedInstances {
   widgetClassPath=&EditorUtilityButton
   parent=HorizontalBox_1
   widget=WriteSelectedInstances
   slot {class=HorizontalBoxSlot, properties {horizontalAlignment=HAlign_Left, verticalAlignment=VAlign_Top, padding=<ltrb 8 8 8 8>}}
  }
  TextBlock_2 {
   widgetClassPath=&TextBlock
   parent=WriteSelectedInstances
   widget=TextBlock_2
   properties {text=Generate List of POI Instances}
   slot {class=ButtonSlot, properties {padding=<ltrb 4 4 4 4>}}
  }
  GoToCSV {
   widgetClassPath=&EditorUtilityButton
   parent=HorizontalBox_1
   widget=GoToCSV
   properties {bIsEnabled=false, toolTipText=Go to Saved CSV in your files}
   slot {class=HorizontalBoxSlot, properties {horizontalAlignment=HAlign_Left, verticalAlignment=VAlign_Top, padding=<ltrb 8 8 8 8>}}
  }
  UEFN_TextBlock_C_60 {
   widgetClassPath=&UEFN_TextBlock
   parent=GoToCSV
   widget=UEFN_TextBlock_C_60
   properties {
    text=Go to CSV
    font {
     bForceMonospaced=false
     bMaterialIsStencil=false
     fontMaterial=null
     fontObject=&NotoSans
     letterSpacing=0
     monospacedWidth=1
     size=15
     skewAmount=0
     typefaceFontName=NotoSans-Bold
     outlineSettings {
      bApplyOutlineToDropShadows=false
      bMiteredCorners=false
      bSeparateFillAlpha=false
      outlineMaterial=null
      outlineSize=0
      outlineColor=<rgba 0 0 0 1>
     }
    }
   }
   slot {class=ButtonSlot}
  }
 ]
}
